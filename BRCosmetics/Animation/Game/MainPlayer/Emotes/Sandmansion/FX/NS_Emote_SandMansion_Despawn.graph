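# Niagara despawn effect for the SandMansion emote, made of delayed, short dust and sand-toned sprite emitters.
# PollenDust bursts sprites outward from a small sphere that then fall under gravity and drag, while Ring_Line spawns tall streaks that drift upward, stretch and fade.
# Both vary color and alpha by Fortnite time of day.
# A third emitter, SmokeLWE, shows no renderers, and the front-end and variant user parameters have no visible use.
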NiagaraSystem /BRCosmetics/Animation/Game/MainPlayer/Emotes/Sandmansion/FX/NS_Emote_SandMansion_Despawn
dependencies {
 usedDataInterfaces [&NiagaraDataInterfaceCurve, &NiagaraDataInterfaceVector2DCurve, &NiagaraDataInterfaceColorCurve]
 usedDynamicInputs [&Multiply_VectorByFloat, &MakeVector]
 usedModules [
  &self:SystemSpawnScript.SystemScriptSource.SystemScriptGraph.NiagaraNodeAssignment_0.SetVariables_#1
  &SystemState
 ]
}
emitters {
 PollenDust {
  emitterUpdateScript [
   &EmitterState(Life Cycle Mode=::NewEnumerator1, Loop Behavior=::NewEnumerator1, Loop Delay=0.5, Loop Duration=0.5, UseLoopDelay=true)
   &SpawnBurst_Instantaneous(Spawn Count=8, Use Spawn Probability=false, UseLoopCountLimit=false)
  ]
  particleSpawnScript [
   &InitializeParticle(
    Color=<rgba 3 3 3 1>
    Lifetime=&UniformRangedFloat(Minimum=0.9, Maximum=0.5)
    Position Offset=<0 0 30>
    Sprite Rotation Mode=::NewEnumerator2
    Sprite Size Mode=::NewEnumerator1
    Uniform Sprite Size Max=100
    Uniform Sprite Size Min=150
    UsePositionOffset=true
    Write Material Random=true
   )
   &DynamicMaterialParameters(
    Index 0 Param 1=&RandomRangeFloat(Minimum=-1, Maximum=1)
    Index 0 Param 2=&RandomRangeFloat(Minimum=0, Maximum=32)
    Module.Index0Float0Write=true
    Module.Index0Float1Write=true
    Module.Index0Float2Write=true
    Module.Index0Float3Write=true
   )
   &CameraOffset(Camera Offset Amount=20)
   &ShapeLocation(MaskSpawnGroup=false, Override Seed=false, Sphere Radius=20, Sphere Surface Distribution=1, UseOwnerScale=true)
   &AddVelocity(
    Module.Use Falloff Distance=false
    Origin Offset=<0 0 -10>
    Override Seed=false
    Velocity Mode=::NewEnumerator1
    Velocity Speed=&RandomRangeFloat(Minimum=125, Maximum=100)
   )
  ]
  particleUpdateScript [
   &ParticleState()
   &Drag(Drag=&RandomRangeFloat(Minimum=2, Maximum=0), Use Linear Drag=true, Use Rotational Drag=true)
   &GravityForce(Gravity=&RandomRangeVector(Minimum=<0 0 -500>, Maximum=<0 0 -700>))
   &SolveForcesAndVelocity(Clamp Velocity=false, Limit Acceleration=false)
   &ScaleColor(
    Scale Alpha=&FloatFromCurve(
     FloatCurve=DataInterface() {propertyValues {Curve {keys [
         {interpMode=RCIM_Cubic, tangentMode=RCTM_User, time=0, value=0}
         <at 0.3224 1 RCIM_Cubic>
         {interpMode=RCIM_Cubic, leaveTangent=-3.8631, time=1, value=0}
        ]}}}
     Scale Curve=&Multiply_Float(
      A=&FNTimeOfDayScalar(Night Scale=0.6, Morning Scale=0.8, Day Scale=0.7, Evening Scale=0.8)
      B=0.3
     )
    )
    Scale RGB=&MakeVectorFromLinearColorRGB(Color=&ColorFromCurve(
      ColorCurve=DataInterface() {propertyValues {
        AlphaCurve {keys [<at 0.13 1>,<at 0.173 1>,<at 0.25 1>,<at 0.375 1>,<at 0.5 1>,<at 0.625 1>,<at 0.75 1>,<at 0.873 1>,<at 0.9078 1>]}
        BlueCurve {keys [<at 0.0514 0.1883>,<at 0.1202 0.2248>,<at 0.25 0.2724>,<at 0.5 0.2754>,<at 0.7338 0.4077>,<at 0.8202 0.2971>,<at 0.8836 0.4589>,<at 0.9613 0.1883>]}
        GreenCurve {keys [<at 0.0514 0.2276>,<at 0.1202 0.3822>,<at 0.25 0.4184>,<at 0.5 0.4897>,<at 0.7338 0.6015>,<at 0.8202 0.6157>,<at 0.8836 0.5652>,<at 0.9613 0.2276>]}
        RedCurve {keys [<at 0.0514 0.2448>,<at 0.1202 0.7344>,<at 0.25 0.7344>,<at 0.5 0.7344>,<at 0.7338 0.7344>,<at 0.8202 0.9855>,<at 0.8836 0.6846>,<at 0.9613 0.2448>]}
       }}
      CurveIndex=Linked() {linkedVariable {name=NPC.FortniteNPC.TimeOfDay_0-1, type=NiagaraFloat}}
     ))
    ScaleA=true
    ScaleRGB=true
   ) {module=ScaleColor001}
   &ScaleSpriteSize(Uniform Curve Sprite Scale=DataInterface() {propertyValues {Curve {keys [
        {arriveTangent=0.6664, interpMode=RCIM_Cubic, leaveTangent=0.6664, tangentMode=RCTM_User, time=0, value=0.777}
        {interpMode=RCIM_Cubic, leaveTangent=-0.4663, time=1, value=1.1262}
       ]}}})
  ]
  rendererClasses [&NiagaraSpriteRendererProperties]
  renderers [{rendererClass=&NiagaraSpriteRendererProperties}]
 }
 Ring_Line {
  emitterUpdateScript [
   &EmitterState(Life Cycle Mode=::NewEnumerator1, Loop Behavior=::NewEnumerator1, Loop Delay=0.58, Loop Duration=0.25, UseLoopDelay=true)
   &SpawnRate(SpawnRate=10, Use Spawn Probability=false)
   &SpawnBurst_Instantaneous(Spawn Count=5, Use Spawn Probability=false, UseLoopCountLimit=false) {module=SpawnBurst_Instantaneous001}
  ]
  particleSpawnScript [
   &InitializeParticle(
    Lifetime Max=0.8
    Lifetime Min=0.5
    Lifetime Mode=::NewEnumerator1
    Position Offset=<0 0 80>
    Sprite Rotation Angle=180
    Sprite Rotation Mode=::NewEnumerator0
    Sprite Size Max=<30 90>
    Sprite Size Min=<30 70>
    Sprite Size Mode=::NewEnumerator2
    UsePositionOffset=true
    Write Material Random=true
   ) {module=InitializeParticle001}
   &ShapeLocation(MaskSpawnGroup=false, Non Uniform Scale=<2 2 5>, Override Seed=false, Sphere Radius=5, Sphere Surface Distribution=1, UseOwnerScale=true)
   &CameraOffset(Camera Offset Amount=20)
   &AddVelocity(Velocity=&RandomRangeVector(Minimum=<0 0 75>, Maximum=<0 0 50>)) {module=AddVelocity001}
  ]
  particleUpdateScript [
   &ParticleState()
   &ScaleSpriteSize(
    Non-Uniform Curve Scale=&RandomRangeVector2D(Minimum=<2 4>, Maximum=<0.5 2>)
    Non-Uniform Curve Sprite Scale=DataInterface() {propertyValues {
      XCurve {keys [{tangentMode=RCTM_User, time=0, value=1}, {tangentMode=RCTM_User, time=1, value=1}]}
      YCurve {
       defaultValue=0
       keys [
        {interpMode=RCIM_Cubic, tangentMode=RCTM_User, time=0, value=0.26}
        {arriveTangent=2.079, interpMode=RCIM_Cubic, leaveTangent=2.079, tangentMode=RCTM_User, time=1, value=1}
       ]
      }
     }}
    Scale Sprite Size Mode=::NewEnumerator3
   )
   &ScaleColor(
    Scale Alpha=&FloatFromCurve(
     FloatCurve=DataInterface() {propertyValues {Curve {keys [
         {arriveTangent=6.8924, interpMode=RCIM_Cubic, leaveTangent=6.8924, tangentMode=RCTM_User, time=0, value=0}
         {arriveTangent=3.3638, interpMode=RCIM_Cubic, leaveTangent=3.3638, tangentMode=RCTM_User, time=0.1, value=0.6}
         {arriveTangent=0.0044, interpMode=RCIM_Cubic, leaveTangent=0.0044, tangentMode=RCTM_User, time=0.2, value=0.75}
         {arriveTangent=-1.0842, interpMode=RCIM_Cubic, leaveTangent=-1.0842, tangentMode=RCTM_User, time=0.4506, value=0.2095}
         {arriveTangent=-0.4598, interpMode=RCIM_Cubic, leaveTangent=-0.4598, tangentMode=RCTM_User, time=1, value=0}
        ]}}}
     Scale Curve=&FNTimeOfDayScalar(Night Scale=1.1, Morning Scale=1.7, Day Scale=2, Evening Scale=1.7)
    )
    Scale RGB=&MakeVectorFromLinearColorRGB(Color=&FNTimeOfDayColor(
      Night Color=&RandomRangeLinearColor(Minimum=<rgba 0.1337 0.1157 0.1059 0>, Maximum=<rgba 0.1059 0.0843 0.0726 0>)
      Morning Color=&RandomRangeLinearColor(Minimum=<rgba 0.4068 0.2264 0.1283 0>, Maximum=<rgba 0.2911 0.162 0.0918 0>)
      Day Color=&RandomRangeLinearColor(Minimum=<rgba 0.5828 0.326 0.1865 0>, Maximum=<rgba 0.3281 0.1658 0.0777 0>)
      Evening Color=&RandomRangeLinearColor(Minimum=<rgba 0.3281 0.2003 0.1308 0>, Maximum=<rgba 0.2541 0.1391 0.0766 0>)
     ))
    ScaleA=true
    ScaleRGB=true
   )
   &DynamicMaterialParameters(
    Index 0 Param 2=50
    Module.Index0Float0Write=true
    Module.Index0Float1Write=true
    Module.Index0Float2Write=true
    Module.Index0Float3Write=true
   )
   &Drag(
    Drag=&FloatFromCurve(
     FloatCurve=DataInterface() {propertyValues {Curve {
        defaultValue=1
        keys [
         {interpMode=RCIM_Cubic, tangentMode=RCTM_User, time=0, value=0}
         {arriveTangent=11.7287, interpMode=RCIM_Cubic, leaveTangent=11.7287, tangentMode=RCTM_User, time=0.1444, value=1}
        ]
       }}}
     Scale Curve=5
    )
    Use Linear Drag=true
    Use Rotational Drag=true
   )
   &SolveForcesAndVelocity(Clamp Velocity=false, Limit Acceleration=false)
  ]
  rendererClasses [&NiagaraSpriteRendererProperties]
  renderers [{rendererClass=&NiagaraSpriteRendererProperties}]
 }
 SmokeLWE {rendererClasses [], renderers []}
}
summary {
 systemName=NS_Emote_SandMansion_Despawn
 emitters [
  {emitterName=Ring_Line, rendererClasses [&NiagaraSpriteRendererProperties]}
  {emitterName=PollenDust, rendererClasses [&NiagaraSpriteRendererProperties]}
  {emitterName=SmokeLWE, rendererClasses []}
 ]
 userVariables [
  User.bIsFrontEnd {type=NiagaraBool}
  User.bIsFrontEndPreview {type=NiagaraBool}
  User.ForwardDot {type=NiagaraFloat, defaultValue=1, description=""}
  User.IsFrontend {type=NiagaraBool}
  User.VariantSwitch {type=NiagaraFloat, defaultValue=0, description=""}
 ]
}
